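# Translucent, two-sided lit Niagara mesh-particle material for the Pizza Party skydiving contrail's panning road.
# A dynamic parameter pans the road texture, whose channels drive yellow/white emissive and edge-faded opacity, and bends the mesh through world position offset.
# An Is FlameShiverContrial scalar, zero by default, blends toward a ShiverFlame-style dirt road with tiled, HSV-tinted dirt and normal-based fake atmosphere lighting.
# The Dive lerp reads one channel twice, so has no effect.
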Material /BRCosmetics/Athena/Effects/Particles/Skydiving/NS_Contrail_PizzaParty/M_Contrail_PizzaPart_PanningRoad
graph {
 properties {
  bUsedWithNiagaraMeshParticles=true
  blendMode=BLEND_Translucent
  pixelDepthOffsetMode=PDOM_Legacy
  twoSided=true
  usedShadingModels=Default Lit
 }
 nodes [
  n1 = ComponentMask(None=TextureCoordinate()) {g=true}
  n2 = OneMinus(None=@n1)
  n3 = Power(Base=SmoothStep(Value=Multiply(A=@n2, B=@n1)) {constMax=0.25}, Exp=ScalarParameter() {defaultValue=1, parameterName=BorderFade_Power}) {constExponent=1}
  n4 = ScalarParameter() {defaultValue=1, parameterName=Diff_TileX}
  n5 = ScalarParameter() {defaultValue=1, parameterName=Diff_TileY}
  n6 = DynamicParameter() {defaultValue=<rgba 0 0 1 1>, paramNames [Panning, BendingUpDown, Dive, BendingLeftRight]}
  n7 = Add(
   A=Multiply(A=AppendVector(A=@n4, B=@n5), B=TextureCoordinate())
   B=MaterialFunctionCall(Y=@n6.Panning) {materialFunction=&MakeFloat2}.Result
  )
  n8 = VectorParameter() {parameterName=Dirt Tile and center scale, defaultValue=<rgba 2 1 1 1>}
  n9 = ComponentMask(None=@n8.RGB) {g=true, r=true}
  n10 = MaterialFunctionCall(Float2=Frac(None=Multiply(A=@n7, B=@n9))) {materialFunction=&BreakOutFloat2Components}
  n11 = Saturate(None=@n10.R)
  n12 = Round(None=MaterialFunctionCall() {materialFunction=&LinearGradient}.UGradient)
  n13 = Add(
   A=MaterialFunctionCall(
    UVs=MaterialFunctionCall(X=@n11, Y=@n10.G) {materialFunction=&MakeFloat2}.Result
    Texture Scale=AppendVector(A=@n8.B, B=@n8.B)
   ) {materialFunction=&ScaleUVsByCenter}.UVs
   B=MaterialFunctionCall(Y=Multiply(A=@n12) {constB=0.5}) {materialFunction=&MakeFloat2}.Result
  )
  n14 = TextureCoordinate()
  n15 = TextureSampleParameter2D!(UVs=@n13, "DDX(UVs)"=DDX(Value=@n14), "DDY(UVs)"=DDY(Value=@n14), Apply View MipBias) {
   mipValueMode=TMVM_Derivative
   parameterName=T_Opacity
   samplerType=SAMPLERTYPE_LinearColor
   texture=/BRCosmetics/Athena/Effects/Particles/Skydiving/NS_Contrail_ShiverFlame_Ace/T_ShiverFlameRoad_Opacity
  }
  n16 = Add(A=@n15.R, B=@n15.R) {constB=0}
  n17 = ScalarParameter() {parameterName=Is FlameShiverContrial}
  LinearInterpolate(A=NamedRerouteUsage(), B=Multiply(A=Multiply(A=@n3, B=@n16), B=ParticleColor().A), Alpha=@n17)
  n18 = TextureSampleParameter2D!(UVs=@n13, Apply View MipBias) {
   parameterName=T_Normal
   samplerType=SAMPLERTYPE_Normal
   texture=/BRCosmetics/Athena/Effects/Particles/Skydiving/NS_Contrail_ShiverFlame_Ace/T_ShiverFlameRoad_normal
  }
  n19 = ConstantBiasScale(None=DotProduct(A=SkyAtmosphereLightDirection(), B=Transform(None=@n18.RGB)))
  LinearInterpolate(
   A=Multiply(A=NamedRerouteUsage(), B=ScalarParameter() {defaultValue=1, parameterName=Emissive scale})
   B=Multiply(
    A=Power(Base=@n19) {constExponent=5}
    B=VectorParameter() {parameterName=Fake Atmos Light Tint, defaultValue=<rgba 0.16 0.1397 0.1316 1>}.RGB
   ) {constB=0.1}
   Alpha=@n17
  )
  n20 = VectorParameter() {parameterName=Dirt HSV A, channelNames {a="", b=Value, g=saturation, r=Hue}, defaultValue=<rgba 0 1 1 0>}
  n21 = VectorParameter() {parameterName=Dirt HSV B, channelNames {a="", b=Value, g=saturation, r=Hue}, defaultValue=<rgba 0 1 1 0>}
  n22 = MaterialFunctionCall(Float3=LinearInterpolate(A=@n20, B=@n21, Alpha=@n15.B)) {materialFunction=&BreakOutFloat3Components}
  n23 = TextureSampleParameter2D(UVs=@n13) {parameterName=T_Dirt, texture=&/T_Seven_PlantersDirt_01_D_Contrail}
  n24 = MaterialFunctionCall(Hue Shift Percentage=@n22.R, Texture=@n23.RGB) {materialFunction=&HueShift}
  n25 = OneMinus(None=@n22.G)
  Multiply(A=Multiply(A=Desaturation(None=@n24.Result, Fraction=@n25), B=@n22.B), B=ScalarParameter() {parameterName=Base color scale})
  n26 = ParticleColor()
  n27 = TextureSampleParameter2D(UVs=@n7) {parameterName=T_Diff, texture=&/T_Dive_PizzaParty_RoadTex}
  n28 = LinearInterpolate(
   A=VectorParameter() {parameterName=YellowColor, defaultValue=<rgba 1 0.6864 0 1>}.RGB
   B=VectorParameter() {parameterName=WhiteColor, defaultValue=<rgba 1 1 1 0>}.RGB
   Alpha=@n27.B
  )
  NamedRerouteDeclaration(None=MaterialFunctionCall(Emissive=Multiply(A=@n26.RGB, B=@n28)) {materialFunction=&MF_FX_EmissiveAdjustments}.Result) {name=Emissive, nodeColor=<rgba 0.4297 1 0 1>}
  n29 = LinearInterpolate(A=@n27.R, B=@n27.R, Alpha=@n6.Dive)
  NamedRerouteDeclaration(None=Multiply(A=@n26.A, B=Multiply(A=@n29, B=@n3))) {name=RoadOpacity, nodeColor=<rgba 1 0 0.75 1>}
  n30 = ComponentMask(None=TextureCoordinate()) {g=true}
  n31 = SmoothStep(Value=Multiply(A=OneMinus(None=@n30), B=@n30)) {constMax=0.25}
  n32 = Saturate(None=Power(Base=OneMinus(None=@n31)) {constExponent=0.5})
  n33 = Multiply(
   A=Transform(None=Constant3Vector() {constant=<rgba 0 0 1 0>}) {transformSourceType=TRANSFORMSOURCE_Instance}
   B=DynamicParameter() {defaultValue=<rgba 0 0 1 1>, paramNames [Panning, BendingUpDown, Dive, BendingLeftRight]}.BendingUpDown
  )
  n34 = ComponentMask(None=TextureCoordinate()) {g=true}
  n35 = Saturate(None=Power(Base=OneMinus(None=@n34)) {constExponent=4})
  n36 = Multiply(
   A=Transform(None=Constant3Vector() {constant=<rgba 1 0 0 0>}) {transformSourceType=TRANSFORMSOURCE_Instance}
   B=DynamicParameter() {defaultValue=<rgba 0 0 1 1>, paramNames [Panning, BendingUpDown, Dive, BendingLeftRight]}.BendingLeftRight
  )
  NamedRerouteDeclaration(None=Add(A=Multiply(A=@n32, B=@n33), B=Multiply(A=@n35, B=@n36))) {name=WPO, nodeColor=<rgba 0 1 0.9297 1>}
  NamedRerouteUsage()
 ]
}
