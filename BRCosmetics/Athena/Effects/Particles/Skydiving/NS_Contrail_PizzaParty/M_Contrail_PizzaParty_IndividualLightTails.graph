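# Translucent Niagara ribbon material for the individual light tails of the Pizza Party skydiving contrail.
# Particle color RGB feeds the shared FX emissive adjustments function.
# Particle alpha is multiplied by a procedural, texture-free UV mask that fades at the ribbon's edges.
# The mask combines saturated half-sine arcs on each UV axis, sharpened by a power along V and smoothstepped along U for short end fades.
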Material /BRCosmetics/Athena/Effects/Particles/Skydiving/NS_Contrail_PizzaParty/M_Contrail_PizzaParty_IndividualLightTails
graph {
 properties {bUsedWithNiagaraRibbons=true, blendMode=BLEND_Translucent, pixelDepthOffsetMode=PDOM_Legacy, usedShadingModels=Default Lit}
 nodes [
  n1 = ParticleColor()
  MaterialFunctionCall(Emissive=@n1.RGB) {materialFunction=&MF_FX_EmissiveAdjustments}
  n2 = TextureCoordinate()
  n3 = Saturate(None=Sine(None=ComponentMask(None=@n2) {g=true}) {period=2})
  n4 = TextureCoordinate()
  n5 = Saturate(None=Sine(None=ComponentMask(None=@n4) {r=true}) {period=2})
  Multiply(A=@n1.A, B=Multiply(A=Power(Base=@n3) {constExponent=3}, B=SmoothStep(Value=@n5) {constMax=0.2}))
 ]
}
