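# Translucent Niagara ribbon and mesh-particle material for the light tails of the Pizza Party skydiving contrail.
# Two panning noise samples drive a boosted emissive, tinted by particle color and a yellow/orange blend selected through a dynamic parameter.
# Opacity combines that noise with sine-based UV edge masks, particle alpha, near-camera fade and an optional edge fade.
# An optional SplineThicken world position offset takes its width from a second dynamic parameter.
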Material /BRCosmetics/Athena/Effects/Particles/Skydiving/NS_Contrail_PizzaParty/M_Contrail_PizzaParty_LightTails
graph {
 properties {
  bUsedWithNiagaraMeshParticles=true
  bUsedWithNiagaraRibbons=true
  blendMode=BLEND_Translucent
  pixelDepthOffsetMode=PDOM_Legacy
  usedShadingModels=Default Lit
 }
 nodes [
  n1 = TextureCoordinate()
  n2 = VectorParameter() {parameterName=UV_Settings_Noise1, defaultValue=<rgba 0.01 0.3 0.2 -0.02>}
  n3 = MaterialFunctionCall(X=@n2.R, Y=@n2.R) {materialFunction=&MakeFloat2}
  n4 = MaterialFunctionCall(X=@n2.B, Y=@n2.B) {materialFunction=&MakeFloat2}
  n5 = Time()
  n6 = TextureSample(UVs=Add(A=Multiply(A=@n1, B=@n3.Result), B=Multiply(A=@n4.Result, B=@n5))) {samplerType=SAMPLERTYPE_Normal, texture=/Game/Global/Textures/Default/Noise/64x64Noise}
  n7 = Add(A=TextureCoordinate()) {constB=0.25}
  n8 = VectorParameter() {parameterName=UV_Settings_Noise2, defaultValue=<rgba 0.02 0.15 0.1 0>}
  n9 = MaterialFunctionCall(X=@n8.R, Y=@n8.R) {materialFunction=&MakeFloat2}
  n10 = MaterialFunctionCall(X=@n8.B, Y=@n8.B) {materialFunction=&MakeFloat2}
  n11 = Time()
  n12 = TextureSample(UVs=Add(A=Multiply(A=@n7, B=@n9.Result), B=Multiply(A=@n10.Result, B=@n11))) {samplerType=SAMPLERTYPE_Normal, texture=&64x64Noise}
  n13 = Saturate(None=Add(A=@n6.G, B=@n12.B))
  Add(A=@n13)
  n14 = ParticleColor()
  n15 = Sine(None=ComponentMask(None=TextureCoordinate()) {g=true}) {period=2}
  n16 = Sine(None=ComponentMask(None=TextureCoordinate()) {r=true}) {period=2}
  n17 = Multiply(A=SmoothStep(Value=Saturate(None=@n15)), B=SmoothStep(Value=Saturate(None=@n16)) {constMax=0.1})
  n18 = Multiply(A=@n6.R, B=@n12.G)
  n19 = Max(
   A=@n18
   B=Multiply(A=@n13, B=DynamicParameter() {defaultValue=<rgba 1 0 1 1>, paramNames [InvertColor,GlowStrength,Param3,Param4]}.GlowStrength)
  )
  n20 = Multiply(
   A=Multiply(A=@n14.A, B=Multiply(A=@n17, B=@n19))
   B=MaterialFunctionCall(Distance=ScalarParameter() {defaultValue=250, parameterName=CameraFadeDist}) {materialFunction=&Near_Camera_Fade}.Result
  )
  n21 = TextureCoordinate()
  n22 = Multiply(A=ComponentMask(None=OneMinus(None=@n21)) {r=true}) {constB=2}
  LinearInterpolate(A=@n20, B=Multiply(A=@n20, B=Saturate(None=@n22)), Alpha=ScalarParameter() {parameterName=UseEdgeFade})
  n23 = DynamicParameter() {parameterIndex=1, defaultValue=<rgba 1 1 0 1>, paramNames [SplineTip,SplineBase,FlipEdgeFade,Param4]}
  LinearInterpolate(
   B=MaterialFunctionCall(WidthBase=@n23.SplineBase, WidthTip=@n23.SplineBase, Expand U or V UV Channel=StaticBoolParameter() {parameterName=ExpandUorV}) {materialFunction=&SplineThicken}.WorldPositionOffset
   Alpha=ScalarParameter() {parameterName=UseSplineThicken}
  )
  n24 = ComponentMask(None=TextureCoordinate()) {r=true}
  n25 = LinearInterpolate(
   B=LinearInterpolate(
    A=VectorParameter() {parameterName=YellowColor, defaultValue=<rgba 1.8538 4.6511 5 1>}.RGB
    B=VectorParameter() {parameterName=OrangeColor, defaultValue=<rgba 1 0.32 0 1>}.RGB
    Alpha=DynamicParameter() {defaultValue=<rgba 1 0 1 1>, paramNames [InvertColor,GlowStrength,Param3,Param4]}.InvertColor
   )
   Alpha=SmoothStep(Value=@n18) {constMax=0.6, constMin=0.15}
  ) {constA=1}
  Multiply(
   A=LinearInterpolate(A=ScalarParameter() {defaultValue=15, parameterName=BaseEmissiveBoost}, Alpha=SmoothStep(Value=@n24) {constMax=0.15}) {constA=3}
   B=Multiply(A=Multiply(A=@n25, B=@n14.RGB) {constA=1}, B=Add(A=@n13))
  )
  Power(Base=@n18)
 ]
}
