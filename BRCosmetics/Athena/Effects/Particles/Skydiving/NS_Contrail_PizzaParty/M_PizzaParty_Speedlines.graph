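# Additive unlit Niagara sprite material for the speedlines in the Pizza Party skydiving contrail.
# Two smoothsteps along V blend tip, mid and trail colors (white-hot, orange, red by default), multiplied by particle color and passed through MF_FX_EmissiveAdjustments.
# A squared, smoothstepped U*(1-U) mask softens both side edges.
# That mask is multiplied by particle alpha and a near-camera fade driven by the CameraFade distance.
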Material /BRCosmetics/Athena/Effects/Particles/Skydiving/NS_Contrail_PizzaParty/M_PizzaParty_Speedlines
graph {
 properties {bUsedWithNiagaraSprites=true, blendMode=BLEND_Additive, pixelDepthOffsetMode=PDOM_Legacy, shadingModel=MSM_Unlit, usedShadingModels=Unlit}
 nodes [
  n1 = ParticleColor()
  n2 = TextureCoordinate()
  n3 = ComponentMask(None=@n2) {g=true}
  n4 = LinearInterpolate(
   A=VectorParameter() {parameterName=Tip_Color, defaultValue=<rgba 5 5 5 1>}.RGB
   B=VectorParameter() {parameterName=Mid_Color, defaultValue=<rgba 4 1.3424 0 1>}.RGB
   Alpha=SmoothStep(Value=@n3) {constMax=0.12}
  )
  n5 = VectorParameter() {parameterName=Trail_Color, defaultValue=<rgba 2 0 0 1>}
  n6 = SmoothStep(Value=@n3) {constMax=0.5, constMin=0.3}
  MaterialFunctionCall(Emissive=Multiply(A=@n1.RGB, B=LinearInterpolate(A=@n4, B=@n5.RGB, Alpha=@n6))) {materialFunction=&MF_FX_EmissiveAdjustments}
  n7 = TextureCoordinate()
  n8 = OneMinus(None=@n7)
  n9 = MaterialFunctionCall(Float2=SmoothStep(Value=Multiply(A=@n7, B=@n8)) {constMax=0.45}) {materialFunction=&BreakOutFloat2Components}
  Multiply(
   A=Multiply(A=@n1.A, B=Multiply(A=@n9.R, B=@n9.R))
   B=MaterialFunctionCall(Distance=ScalarParameter() {defaultValue=100, parameterName=CameraFade}) {materialFunction=&Near_Camera_Fade}.Result
  )
  OneMinus(None=@n2)
 ]
}
