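# Skydiving contrail cosmetic "PizzaParty" that attaches ribbons, a road mesh and speed lines to the player's skeletal mesh.
# Ribbons splits particles into four bone-attached trails, two light blue and two red.
# Road follows a bone, feeding panning and velocity-based bend from local SetRoadBending into its material, while Speedlines spawns streak sprites on a torus around that bone.
# ParaGlideLeanAlpha modulates all three emitters, with front-end modifiers adjusting motion for menus.
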NiagaraSystem /BRCosmetics/Athena/Effects/Particles/Skydiving/NS_Contrail_PizzaParty/NS_Contrail_PizzaParty
dependencies {
 usedDataInterfaces [&NiagaraDataInterfaceCurve, &NiagaraDataInterfaceColorCurve, &NiagaraDataInterfaceVector2DCurve]
 usedDynamicInputs [/Niagara/DynamicInputs/Rotation/InvertQuaternion]
 usedModules [
  /Game/Effects/Niagara/Modules/Skydiving/FN_DivingAttributes/v5/FN_DivingAttributes
  &self:SystemSpawnScript.SystemScriptSource.SystemScriptGraph.NiagaraNodeAssignment_1.SetVariables_#1
  &self:SystemSpawnScript.SystemScriptSource.SystemScriptGraph.NiagaraNodeAssignment_0.SetVariables_#2
  &SystemState
 ]
}
emitters {
 Ribbons {
  emitterUpdateScript [
   &EmitterState(Inactive Response=::NewEnumerator1, Life Cycle Mode=::NewEnumerator1, Loop Duration=0, UseLoopDelay=false)
   /Game/Effects/Niagara/Modules/Skydiving/FN_DivingAttributes/FN_EmitterDivingAttributes_v1(Bone Index=1, SkeletalMesh:NiagaraDataInterfaceSkeletalMesh=$User.SkeletalMesh)
   &FN_SpawnRibbon_v2(Number of Ribbons=4, Spawn Mode=::NewEnumerator0, SpawnRate=128)
   /SetVariables(Emitter.TrailFade:NiagaraFloat=&FloatFromCurve(
     FloatCurve=DataInterface() {propertyValues {Curve {keys [{interpMode=RCIM_Cubic, tangentMode=RCTM_User, time=0, value=0}, {interpMode=RCIM_Cubic, tangentMode=RCTM_User, time=0.75, value=1}]}}}
     CurveIndex=$Emitter.Age
    )) {isSetParameters=true}
  ]
  particleSpawnScript [
   &PartitionParticles(Number Of Partitions=4)
   &FN_InitializeRibbon_v3(
    Color:LinearColor=&LerpLinearColors(
     StartColor=<rgba 0.3664 0.7106 1 1>
     EndColor=<rgba 1 0 0 1>
     LerpFactor=&ClampFloat(Float=&MakeFloatFromInt(Int=/Niagara/DynamicInputs/Subtract/Subtract_Integer(A:NiagaraInt32=$Particles.PartitionParticles.Partition, B:NiagaraInt32=1)))
    )
    Custom Offset=true
    Enable Skeletal Position=true
    Lifetime:NiagaraFloat=&Lerp_Float(A=0.45, B=0.75, Alpha=$User.ParaGlideLeanAlpha)
    Local Vector Offset:Vector3f=<0 1 0>
    Ribbon Width:NiagaraFloat=&Lerp_Float(A=15, B=15, Alpha=$User.ParaGlideLeanAlpha)
   )
   &SkeletalMeshLocation(
    Apply Sampled Orientation to Mesh Particles=true
    Bone / Socket Index=$Particles.PartitionParticles.Partition
    Bone Sampling Mode=::NewEnumerator4
    Inherit Velocity=false
    Interpolate Mesh Sampling=false
    MaskSpawnGroup=false
    Sampled Position Offset Coordinate Space:ENiagaraMeshTransforms=::NewEnumerator3
    Skeletal Mesh=$User.SkeletalMesh
   )
   &AddVelocity(
    Scale Added Velocity=&Lerp_Vector(A=<1 1 1>, B=<0 0.5 1>, Alpha=$System.Dive)
    Velocity=&FN_FrontEndVector3Modifier(Gameplay=<0 0 1500>, Front End=&Lerp_Vector(A=<0 0 1500>, B=<1 -2500 750>, Alpha=$System.Dive))
   )
   &DynamicMaterialParameters(
    Index 0 Param 1=&ModuloFloat(A=&MakeFloatFromInt(Int=$Particles.PartitionParticles.Partition), B=2)
    Module.Index0Param0Write=true
    Module.Index0Param1Write=true
    Module.Index0Param2Write=true
    Module.Index0Param3Write=true
   )
  ]
  particleUpdateScript [
   &InheritSourceMovement(Applied Force Scale=<0.5 0.5 1>, Applied Position Scale=<0.5 0.5 1>, ApplyForce=false, ApplyPosition=true, ApplyPositionFalloff=false)
   &SolveForcesAndVelocity(Clamp Velocity=false, Limit Acceleration=false)
   &FN_UpdateRibbon(Detach Locked Particle=true, Enable Skeletal Position=true)
   &ScaleColor(
    Color Value To Scale=$Particles.Color
    Scale Alpha=&Lerp_Float(
     A=&FloatFromCurve(FloatCurve=DataInterface() {propertyValues {Curve {keys [{interpMode=RCIM_Cubic, tangentMode=RCTM_User, time=0.2, value=1}, {interpMode=RCIM_Cubic, tangentMode=RCTM_User, time=0.25, value=0}]}}})
     B=&FloatFromCurve(FloatCurve=DataInterface() {propertyValues {Curve {keys [{interpMode=RCIM_Cubic, tangentMode=RCTM_User, time=0.1, value=1}, {interpMode=RCIM_Cubic, tangentMode=RCTM_User, time=0.15, value=0}]}}})
     Alpha=&ClampFloat(Float=&MakeFloatFromInt(Int=&Subtract_Integer(A:NiagaraInt32=$Particles.PartitionParticles.Partition, B:NiagaraInt32=1)))
    )
    ScaleA=true
    ScaleRGB=true
   )
   /SetVariables(Particles.RibbonWidth=&Multiply_Float(
     A=$Particles.Initial.RibbonWidth
     B=&Lerp_Float(
      A=&FloatFromCurve(FloatCurve=DataInterface() {propertyValues {Curve {keys [<at 0 1>,<at 1 0>]}}})
      B=&FloatFromCurve(FloatCurve=DataInterface() {propertyValues {Curve {keys [<at 0 1>,<at 0.5 2>]}}})
      Alpha=$User.ParaGlideLeanAlpha
     )
    )) {isSetParameters=true}
   &DynamicMaterialParameters(
    Index 0 Param 1=&ClampFloat(Float=&MakeFloatFromInt(Int=&Subtract_Integer(A:NiagaraInt32=$Particles.PartitionParticles.Partition, B:NiagaraInt32=1)))
    Index 0 Param 2=&Lerp_Float(A=0.05, B=0.2, Alpha=$User.ParaGlideLeanAlpha)
    Module.Index0Param0Write=true
    Module.Index0Param1Write=true
    Module.Index0Param2Write=true
    Module.Index0Param3Write=true
   ) {module=DynamicMaterialParameters001}
  ]
  rendererClasses [&NiagaraRibbonRendererProperties]
  renderers [{rendererClass=&NiagaraRibbonRendererProperties}]
 }
 Road {
  emitterUpdateScript [
   &EmitterState(Life Cycle Mode=::NewEnumerator1, Loop Behavior=::NewEnumerator1, Loop Duration=1.6644, UseLoopDelay=false)
   &SpawnBurst_Instantaneous(Use Spawn Probability=false, UseLoopCountLimit=false)
   &SpawnBurst_Instantaneous(Spawn Time=1.6644, Use Spawn Probability=false, UseLoopCountLimit=false) {module=SpawnBurst_Instantaneous001}
   /SetVariables(Emitter.PanningFloat:NiagaraFloat=&Add_Float(
     A=$Emitter.PanningFloat
     B=&Multiply_Float(A=$Engine.DeltaTime, B=&Lerp_Float(A=-1, B=-2.5, Alpha=$User.ParaGlideLeanAlpha))
    )) {isSetParameters=true}
  ]
  particleSpawnScript [
   &InitializeParticle(
    Color Mode=::NewEnumerator0
    Lifetime=5
    Mesh Scale Mode=::NewEnumerator4
    Sprite Size Mode=::NewEnumerator3
    Uniform Sprite Size=550
    UsePositionOffset=false
    Write Material Random=true
   )
   &SkeletalMeshLocation(
    Bone / Socket Index=5
    Bone Sampling Mode=::NewEnumerator4
    MaskSpawnGroup=false
    Orientation Sampling=::NewEnumerator1
    Sampled Position Offset=<10 35 0>
    Skeletal Mesh=$User.SkeletalMesh
   ) {module=SkeletalMeshLocation001}
   /SetVariables(Particles.MeshOrientation=&MultiplyQuaternion(
     Quaternion A=&MultiplyQuaternion(
      Quaternion A=$Particles.SkeletalMeshLocation001.SampledRotation
      Quaternion B=&MakeQuatFromAxisAngle(Axis=<0 1 0>, AngleInDegrees=90)
     )
     Quaternion B=&MakeQuatFromAxisAngle(Axis=<1 0 0>, AngleInDegrees=90)
    )) {isSetParameters=true}
   /SetVariables() {isSetParameters=true}
  ]
  particleUpdateScript [
   &ParticleState(Kill Particles When Lifetime Has Elapsed=false)
   &ScaleColor(
    Scale Alpha=&FloatFromCurve(
     FloatCurve=DataInterface() {propertyValues {Curve {keys [{interpMode=RCIM_Cubic, tangentMode=RCTM_User, time=0, value=0}, {interpMode=RCIM_Cubic, tangentMode=RCTM_User, time=1, value=1}]}}}
     CurveIndex=$Emitter.Age
    )
    ScaleA=true
    ScaleRGB=true
   )
   /SetVariables(Particles.PreviousPos:NiagaraPosition=$Particles.Position) {isSetParameters=true}
   &SkeletalMeshLocation(
    Bone / Socket Index=5
    Bone Sampling Mode=::NewEnumerator4
    MaskSpawnGroup=false
    Orientation Sampling=::NewEnumerator1
    Sampled Position Offset=<10 33 0>
    Skeletal Mesh=$User.SkeletalMesh
   )
   /SetVariables(Particles.OrientationGoal:Quat4f=&MultiplyQuaternion(
     Quaternion A=&MultiplyQuaternion(
      Quaternion A=$Particles.SkeletalMeshLocation.SampledRotation
      Quaternion B=&MakeQuatFromAxisAngle(Axis=<0 1 0>, AngleInDegrees=90)
     )
     Quaternion B=&MakeQuatFromAxisAngle(Axis=<1 0 0>, AngleInDegrees=90)
    )) {isSetParameters=true}
   &self:SetRoadBending(CurrentPos:NiagaraPosition=$Particles.Position, MaxBending=125, RestSpeed=3, VelDivider=4) {module=SetRoadBending}
   /SetVariables(Particles.MeshOrientation=&LerpQuaternion(
     Quaternion A:Quat4f=$Particles.MeshOrientation
     Quaternion B:Quat4f=$Particles.OrientationGoal
     Lerp Factor:NiagaraFloat=&FN_FrontEndFloatModifier(Gameplay=&Multiply_Float(A=$Engine.DeltaTime, B=3), Front End=1)
     Enable Spherical Linear Interpolation:NiagaraBool=false
    )) {enabled=false, isSetParameters=true}
   /SetVariables(Particles.MeshOrientation=$Particles.OrientationGoal) {isSetParameters=true}
   &DynamicMaterialParameters(
    Index 0 Param 1=$Emitter.PanningFloat
    Index 0 Param 2=&Lerp_Float(A=50, B=-20, Alpha=$User.ParaGlideLeanAlpha)
    Index 0 Param 3=$User.ParaGlideLeanAlpha
    Index 0 Param 4=$Particles.Roadbending
    Module.Index0Param0Write=true
    Module.Index0Param1Write=true
    Module.Index0Param2Write=true
    Module.Index0Param3Write=true
   )
  ]
  rendererClasses [&NiagaraMeshRendererProperties]
  renderers [{rendererClass=&NiagaraMeshRendererProperties, rendererIndex=0}]
 }
 Speedlines {
  emitterUpdateScript [
   &EmitterState()
   &SpawnRate(SpawnRate=20, Use Spawn Probability=true) {module=SpawnRate001}
  ]
  particleSpawnScript [
   &InitializeParticle(
    Color=&ColorFromCurve(
     ColorCurve=DataInterface() {propertyValues {
       AlphaCurve {keys [<at 0 1 RCIM_Cubic>]}
       BlueCurve {keys [<at 0.4214 2 RCIM_Cubic>]}
       GreenCurve {keys [<at 0.4214 2 RCIM_Cubic>]}
       RedCurve {keys [<at 0.4214 2 RCIM_Cubic>]}
      }}
     CurveIndex=&RandomRangeFloat(Minimum=0, Maximum=1)
    )
    Lifetime Max=0.35
    Lifetime Min=0.2
    Lifetime Mode=::NewEnumerator1
    Position Offset=&FN_FrontEndVector3Modifier(Gameplay=<0 0 0>, Front End=<0 50 0>)
    Sprite Size Max=<5 250>
    Sprite Size Min=<3 150>
    Sprite Size Mode=::NewEnumerator2
    UsePositionOffset=true
    Write Material Random=true
   )
   &SkeletalMeshLocation(
    Bone / Socket Index=5
    Bone Sampling Mode=::NewEnumerator4
    MaskSpawnGroup=false
    Orientation Sampling=::NewEnumerator1
    Sampled Position Offset=<0 50 0>
    Skeletal Mesh=$User.SkeletalMesh
   ) {module=SkeletalMeshLocation001}
   &ShapeLocation(
    Additional Quaternion Rotation=true
    Additional Yaw / Pitch / Roll=<0 0.25 0>
    MaskSpawnGroup=false
    Offset=&Add_Vector(
     A=&Multiply_Vector(
      A=&RotateVectorByQuaternion(Vector To Rotate=<1 0 0>, Quaternion=$Particles.SkeletalMeshLocation001.SampledRotation)
      B=&VectorFromFloat(Value=300)
     )
     B=&Multiply_Vector(
      A=&RotateVectorByQuaternion(Vector To Rotate=<0 0 1>, Quaternion=$Particles.SkeletalMeshLocation001.SampledRotation)
      B=&Multiply_VectorByFloat(
       Vector=&VectorFromFloat(Value=&MakeFloatFromVector(
         Vector=&TransformVector(Vector=$Engine.Owner.Velocity, Source Space=::World, Destination Space=::Local)
         Channel=::NewEnumerator0
        ))
       Float=&FN_FrontEndFloatModifier(Gameplay=-0.1, Front End=0)
      )
     )
    )
    Offset Coordinate Space=::Local
    Offset Mode=::NewEnumerator0
    Rotation Mode=::NewEnumerator3
    Rotation Quaternion=$Particles.SkeletalMeshLocation001.SampledRotation
    Shape Primitive=::NewEnumerator3
    Torus Distribution Mode:ENiagaraTorusDistributionMode=::NewEnumerator1
    U Position=&RandomRangeFloat(Minimum=0.5, Maximum=1)
    UseOwnerScale=true
    V Position=1
   )
   &AddVelocity(Velocity=&Multiply_VectorByFloat(
     Vector=&NormalizeVector(
      Vector To Normalize=&RotateVectorByQuaternion(Vector To Rotate=<1 0 0>, Quaternion=$Particles.SkeletalMeshLocation001.SampledRotation)
      Safe Normalize=true
      Safe Fallback Threshold=0
      Safe Fallback Vector=<0 0 0.001>
     )
     Float=&RandomRangeFloat(Minimum=-1200, Maximum=-1500)
    ))
   &AddVelocity(
    Velocity=&NormalizeVector(Vector To Normalize=$Particles.ShapeLocation.ShapeNormal, Safe Normalize=true, Safe Fallback Threshold=0, Safe Fallback Vector=<0 0 0.001>)
    Velocity Speed Scale=&RandomRangeFloat(Minimum=250, Maximum=300)
   ) {module=AddVelocity001}
   &AddVelocity(
    Velocity=&Multiply_Vector(
     A=&RotateVectorByQuaternion(Vector To Rotate=<0 0 1>, Quaternion=$Particles.SkeletalMeshLocation001.SampledRotation)
     B=&Multiply_VectorByFloat(
      Vector=&VectorFromFloat(Value=&MakeFloatFromVector(
        Vector=&TransformVector(Vector=$Engine.Owner.Velocity, Source Space=::World, Destination Space=::Local)
        Channel=::NewEnumerator0
       ))
      Float=0.5
     )
    )
    Velocity Speed Scale=&FN_FrontEndFloatModifier(Gameplay=1, Front End=0)
   ) {module=AddVelocity002}
   /SetVariables() {isSetParameters=true}
  ]
  particleUpdateScript [
   &ParticleState()
   /SetVariables(Particles.PreviousPos:NiagaraPosition=$Particles.Position) {isSetParameters=true}
   &ScaleSpriteSize(Scale Factor=&Vector2DFromCurve(
     Vector2Curve=DataInterface() {propertyValues {XCurve {keys [<at 0 0.5 RCIM_Cubic>,<at 0.5 1 RCIM_Cubic>]}, YCurve {keys [<at 0 0 RCIM_Cubic>,<at 0.5 1 RCIM_Cubic>]}}}
     ScaleCurve=&Vector2DFromFloat(Value=1)
    ))
   &AccelerationForce(Acceleration=&Multiply_VectorByFloat(
     Vector=&Multiply_Vector(A=$Engine.Owner.Velocity, B=<1 1 0>)
     Float=&FN_FrontEndFloatModifier(Gameplay=1, Front End=0)
    ))
   &SolveForcesAndVelocity(Clamp Velocity=false, Limit Acceleration=false)
   &self:SetRoadBending(CurrentPos:NiagaraPosition=$Particles.Position, MaxBending=125, RestSpeed=3, VelDivider=4) {module=SetRoadBending}
   &ScaleColor(
    Scale Alpha=&Multiply_Float(
     A=&FloatFromCurve(FloatCurve=DataInterface() {propertyValues {Curve {keys [
          <at 0 0 RCIM_Cubic>
          <at 0.3759 1.0207>
          {arriveTangent=-2.221, interpMode=RCIM_Cubic, leaveTangent=-2.221, tangentMode=RCTM_User, time=1, value=0}
         ]}}})
     B=&FloatFromCurve(
      FloatCurve=DataInterface() {propertyValues {Curve {keys [
          <at 0 0 RCIM_Cubic>
          <at 0.5 0>
          {arriveTangent=5.9607, interpMode=RCIM_Cubic, leaveTangent=5.9607, tangentMode=RCTM_User, time=0.8, value=1}
         ]}}}
      CurveIndex=$User.ParaGlideLeanAlpha
     )
    )
    ScaleA=true
    ScaleRGB=true
   )
  ]
  rendererClasses [&NiagaraSpriteRendererProperties]
  renderers [{rendererClass=&NiagaraSpriteRendererProperties}]
 }
}
summary {
 systemName=NS_Contrail_PizzaParty
 emitters [
  {emitterName=Ribbons, rendererClasses [&NiagaraRibbonRendererProperties]}
  {emitterName=Road, rendererClasses [&NiagaraMeshRendererProperties]}
  {emitterName=Speedlines, rendererClasses [&NiagaraSpriteRendererProperties]}
 ]
 userVariables [
  User.bIsFrontEnd {type=NiagaraBool}
  User.bIsFrontEndPreview {type=NiagaraBool}
  User.ForwardDot {type=NiagaraFloat, defaultValue=0, description=""}
  User.ParaGlideLeanAlpha {type=NiagaraFloat, defaultValue=1, description=""}
  User.PlayerSpeed {type=NiagaraFloat, defaultValue=0, description=""}
  User.RibbonIDOffset {type=NiagaraInt32}
  User.RightDot {type=NiagaraFloat, defaultValue=0, description=""}
  User.SkeletalMesh {
   type=NiagaraDataInterfaceSkeletalMesh
   defaultValue {
    struct=NiagaraExt_VariableValue_DataInterface
    value {dataInterface=&self:NiagaraDataInterfaceSkeletalMesh_0}
   }
  }
 ]
}
